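# Translucent unlit material for the Ripped Harvester pickaxe blade, used on Niagara mesh particles to draw a fiery lava swipe.
# The LavaSwipe dynamic parameter slides noise-distorted UVs through a sine, forming a moving band colored by a flame curve atlas.
# Two panning noises multiply to break up the emissive and an alpha-style term, both modulated by particle color.
# A sharpened band, masked by UV height and noise, also scales a vertex-normal offset term.
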Material /BRCosmetics/Weapons/FORT_Melee/Pickaxe_RippedHarvester/FX/M_RippedHarvester_Blade
graph {
 properties {
  bUsedWithNiagaraMeshParticles=true
  blendMode=BLEND_Translucent
  pixelDepthOffsetMode=PDOM_Legacy
  shadingModel=MSM_Unlit
  usedShadingModels=Unlit
 }
 nodes [
  Constant() {r=0.5}
  n1 = TextureCoordinate()
  n2 = TextureCoordinate()
  n3 = TextureSampleParameter2D(UVs=Panner(Coordinate=Multiply(A=@n2)) {speedX=0.1, speedY=0.2}) {
   parameterName=Tex3
   samplerType=SAMPLERTYPE_Masks
   texture=/BRCosmetics/Characters/Player/Male/Large/Bodies/M_LRG_Clash_V/Textures/Simple_Perlin_M
  }
  n4 = Multiply(A=@n3.R) {constB=0.4}
  n5 = Add(
   A=Add(A=Add(A=@n1, B=@n4) {constB=0}) {constB=-0.5}
   B=LinearInterpolate(Alpha=DynamicParameter() {paramNames [LavaSwipe,Param2,Param3,Param4]}.LavaSwipe) {constA=-0.5, constB=1.5}
  )
  n6 = Sine(None=Multiply(A=ComponentMask(None=@n5) {g=true}) {constB=0.5})
  n7 = Power(Base=Multiply(A=SmoothStep(Value=@n6) {constMin=0.2}) {constB=0.8}) {constExponent=1.1}
  n8 = CurveAtlasRowParameter(CurveTime=Saturate(None=@n7)) {
   atlas=/Game/Weapons/FORT_Melee/Pickaxe_RippedHarvester/FX/CurveAtlas_RippedHarvester_FlameColor
   curve=/Game/Weapons/FORT_Melee/Pickaxe_RippedHarvester/FX/Curve_RippedHarvester_FireColor
   parameterName=ColorCurve
  }
  n9 = ParticleColor()
  n10 = Multiply(A=TextureCoordinate()) {constB=0.5}
  n11 = Multiply(A=TextureCoordinate()) {constB=0.3}
  n12 = Multiply(
   A=TextureSampleParameter2D(UVs=Panner(Coordinate=@n10) {speedX=0.06, speedY=0.12}) {parameterName=Tex1, samplerType=SAMPLERTYPE_LinearColor, texture=/EpicBaseTextures/Noises/T_Perlin_Noise_D}.RGB
   B=TextureSampleParameter2D(UVs=Panner(Coordinate=@n11) {speedX=0.15, speedY=-0.1}) {parameterName=Tex2, texture=&AtmosphericCloudNoise01}.RGB
  )
  n13 = LinearInterpolate(Alpha=@n12) {constA=0.015}
  MaterialFunctionCall(Emissive=Multiply(A=@n8, B=Multiply(A=@n9.RGB, B=@n13))) {materialFunction=&MF_FX_EmissiveAdjustments}
  MaterialFunctionCall(Float4=@n7) {materialFunction=&BreakOutFloat4Components}
  n14 = LinearInterpolate(A=@n12) {constAlpha=0.8}
  Multiply(A=Multiply(A=@n14, B=@n6), B=ParticleColor().A)
  n15 = TextureCoordinate()
  n16 = Saturate(None=Add(A=ComponentMask(None=@n15) {g=true}) {constB=0.3})
  n17 = VertexNormalWS()
  n18 = Saturate(None=Power(Base=@n6) {constExponent=40})
  Multiply(A=Multiply(A=Multiply(A=@n16, B=@n17), B=Multiply(A=@n14, B=@n18))) {constB=5}
  ParticleRelativeTime()
  Sine(None=Multiply(A=Time()) {constB=0.3})
  SmoothStep(Value=@n6)
  SmoothStep(Value=@n6) {constMin=0.983}
 ]
}
