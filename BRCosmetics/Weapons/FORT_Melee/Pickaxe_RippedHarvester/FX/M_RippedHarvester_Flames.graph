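# Translucent unlit Niagara sprite material for the flame effect on the Ripped Harvester pickaxe.
# Polar UVs drive a panning lava noise texture, warped by panning Perlin noise scaled with radius.
# A radial mask times particle alpha erodes the noise, which indexes a curve atlas for flame color, tinted by particle color.
# A second erosion of the same noise uses a different softness and is scaled down.
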Material /BRCosmetics/Weapons/FORT_Melee/Pickaxe_RippedHarvester/FX/M_RippedHarvester_Flames
graph {
 properties {bUsedWithNiagaraSprites=true, blendMode=BLEND_Translucent, pixelDepthOffsetMode=PDOM_Legacy, shadingModel=MSM_Unlit, usedShadingModels=Unlit}
 nodes [
  Constant()
  n1 = MaterialFunctionCall!(UV=TextureCoordinate(), Twist Center, Twist Edges) {materialFunction=/Game/Accessories/FORT_Backpacks/Backpack_F_MED_Celestial_Glow/FX/Materials/MF_PolarUV}
  n2 = MaterialFunctionCall(Float2=@n1.Polar UVs) {materialFunction=&BreakOutFloat2Components}
  n3 = AppendVector(A=Multiply(A=@n2.R), B=Multiply(A=Power(Base=@n2.G) {constExponent=0.666}) {constB=0.4})
  n4 = TextureCoordinate()
  n5 = TextureSampleParameter2D(UVs=Panner(Coordinate=Multiply(A=@n4)) {speedX=-0.06, speedY=0.23}) {
   parameterName=TexDistort
   samplerType=SAMPLERTYPE_Masks
   texture=/BRCosmetics/Characters/Player/Male/Large/Bodies/M_LRG_Clash_V/Textures/Simple_Perlin_M
  }
  n6 = Multiply(A=@n5.R) {constB=0.1}
  n7 = TextureSample!(UVs=Add(A=Panner(Coordinate=@n3) {speedY=-0.1}, B=Multiply(A=@n2.G, B=@n6)), Tex, "DDX(UVs)"=@n1.DDX, "DDY(UVs)"=@n1.DDY, Apply View MipBias) {
   mipValueMode=TMVM_Derivative
   samplerType=SAMPLERTYPE_LinearGrayscale
   texture=/BRCosmetics/Weapons/FORT_Melee/Pickaxe_RenegadeRaider_Spark/FX/Materials/T_Pickaxe_RenegadeRaider_Spark_LavaNoise
  }
  n8 = Saturate(None=Power(Base=OneMinus(None=@n2.G)))
  n9 = Saturate(None=Power(Base=OneMinus(None=@n2.G)) {constExponent=10})
  n10 = OneMinus(None=Multiply(A=Subtract(A=@n8, B=@n9), B=ParticleColor().A))
  n11 = Constant() {r=4}
  n12 = CurveAtlasRowParameter(CurveTime=Multiply(A=MaterialFunctionCall!(Gradient=@n7.R, Erosion Amount=@n10, Softness=@n11) {materialFunction=/BRCosmetics/Athena/Effects/Particles/Skydiving/NS_Contrail_DarkAndLight/Functions/MF_Erosion}.Increased Range) {constB=0.85}) {
   atlas=/Game/Weapons/FORT_Melee/Pickaxe_RippedHarvester/FX/CurveAtlas_RippedHarvester_FlameColor
   curve=/Game/Weapons/FORT_Melee/Pickaxe_RippedHarvester/FX/Curve_RippedHarvester_FireColor
   parameterName=ColorCurve
  }
  MaterialFunctionCall(Emissive=Multiply(A=ParticleColor().RGB, B=@n12)) {materialFunction=&MF_FX_EmissiveAdjustments}
  MaterialFunctionCall() {materialFunction=&Lerp_Multiple_Float3}
  Multiply(A=MaterialFunctionCall(Gradient=@n7.R, Erosion Amount=@n10, Softness=Constant() {r=0.5}) {materialFunction=&MF_Erosion}.Increased Range) {constB=0.4}
 ]
}
